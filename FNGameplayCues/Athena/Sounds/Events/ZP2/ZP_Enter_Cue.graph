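# ZP_Enter_Cue is a sound cue for an Athena ZP2 event that plays a ZP portal activation sound, named as the enter sound.
# A single wave player feeds the output directly, with no randomization or modulation.
# It uses the ZP launch attenuation and the gadgets and items sound class.
# A concurrency override allows one instance at a time and blocks new plays rather than cutting off the current one.
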SoundCue /FNGameplayCues/Athena/Sounds/Events/ZP2/ZP_Enter_Cue
root=WavePlayer_0
nodes [{id=WavePlayer_0, properties {soundWaveAssetPtr=&/ZP_Portal_Activate_01}}]
properties {
 attenuationSettings=&/ZP_Launch_Att
 bOverrideConcurrency=true
 maxDistance=15000
 soundClassObject=&Fort_Gadgets_Items
 volumeMultiplier=1
 concurrencyOverrides {maxCount=1, resolutionRule=PreventNew}
}
