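# Sound cue for the ZP2 event portal exit, layering a pre-exit sound with the main exit sound.
# After a short delay, a mixer plays the pre-exit wave at reduced volume and the exit wave slightly later at full volume, under a modulator that varies pitch.
# It uses launch attenuation and the gadgets/items sound class, primes on load, and allows one instance, blocking new plays while active.
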SoundCue /FNGameplayCues/Athena/Sounds/Events/ZP2/ZP_Exit_Cue
root=Delay_1
nodes [
 {id=Delay_1, children [Modulator_0], properties {delayMax=0.25, delayMin=0.25}}
 {id=Delay_2, children [WavePlayer_0], properties {delayMax=0.2, delayMin=0.2}}
 {id=Mixer_0, children [WavePlayer_1,Delay_2], properties {inputVolume [0.5,1]}}
 {id=Modulator_0, children [Mixer_0], properties {pitchMin=0.9}}
 {id=WavePlayer_0, properties {soundWaveAssetPtr=&/ZP_Portal_Exit_01}}
 {id=WavePlayer_1, properties {soundWaveAssetPtr=&/ZP_Portal_PreExit_01}}
]
properties {
 attenuationSettings=&/ZP_Launch_Att
 bOverrideConcurrency=true
 bPrimeOnLoad=true
 maxDistance=15000
 soundClassObject=&Fort_Gadgets_Items
 volumeMultiplier=1
 concurrencyOverrides {maxCount=1, resolutionRule=PreventNew}
}
