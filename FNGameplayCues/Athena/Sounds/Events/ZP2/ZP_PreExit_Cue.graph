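# ZP_PreExit_Cue is an Athena gameplay-cue sound for the ZP2 event, named for a pre-exit moment.
# Its node graph is empty with no root, so it has no sound source to play.
# It still carries playback settings: the shared ZP_Launch attenuation, the Fort_Gadgets_Items sound class, reduced volume and priming on load.
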SoundCue /FNGameplayCues/Athena/Sounds/Events/ZP2/ZP_PreExit_Cue
root=null
nodes []
properties {
 attenuationSettings=&/ZP_Launch_Att
 bPrimeOnLoad=true
 maxDistance=15000
 soundClassObject=&Fort_Gadgets_Items
 volumeMultiplier=0.7
}
