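# First-person sound cue for ZP portal travel in the ZP2 event, layering a portal enter sound with a travel loop.
# A Mixer plays the enter wave at full volume alongside the quieter looping travel wave.
# The loop starts after a fixed half-second delay, and an Enveloper volume curve fades it in gradually over about three seconds.
# It uses trap attenuation and the gadgets/items sound class, and is primed on load.
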SoundCue /FNGameplayCues/Athena/Sounds/Events/ZP2/ZP_Travel_1P_Cue
root=Mixer_0
nodes [
 {id=Delay_0, children [WavePlayer_2], properties {delayMax=0.5, delayMin=0.5}}
 {
  id=Enveloper_0
  children [Delay_0]
  properties {
   pitchCurve {editorCurveData {keys [{time=0}]}}
   volumeCurve {editorCurveData {keys [<at 0 0>,<at 0.25 0.0007>,<at 0.562 0.0209>,<at 1.0682 0.375>,<at 1.3981 0.4226>,<at 1.6907 0.4226>,<at 1.9397 0.4445>,<at 2.2945 0.5119>,<at 2.7365 0.6131>,<at 3.0042 0.6786>]}}
  }
 }
 {id=Mixer_0, children [WavePlayer_1,Enveloper_0], properties {inputVolume [1,0.4]}}
 {id=WavePlayer_1, properties {soundWaveAssetPtr=&/ZP_Portal_Enter_01}}
 {id=WavePlayer_2, properties {bLooping=true, soundWaveAssetPtr=&/ZP_Portal_Travel_Loop_01}}
]
properties {
 attenuationSettings=&Fort_Traps_Att
 bPrimeOnLoad=true
 maxDistance=6000
 soundClassObject=&Fort_Gadgets_Items
 volumeMultiplier=1
}
