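# Empty SoundCue named as the third-person travel sound for ZP2, under Athena gameplay-cue event sounds.
# It has no root or sound nodes, so it plays no audio as it stands.
# Only cue-level settings remain: gadget sound class, trap attenuation and priming on load.
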SoundCue /FNGameplayCues/Athena/Sounds/Events/ZP2/ZP_Travel_3P_Cue
root=null
nodes []
properties {
 attenuationSettings=&Fort_Traps_Att
 bPrimeOnLoad=true
 maxDistance=6000
 soundClassObject=&Fort_Gadgets
 volumeMultiplier=1
}
